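# Post-process AnimBP for the DairyString WadePuck male head, which counters neck and head rotation and adds secondary motion to its flap bones.
# Each update reads neck_01, neck_02 and head rotation deltas from reference pose, negating them scaled by per-axis percentages.
# Additive bone-space Modify Bone nodes apply these to the input pose, then three chained spring-driven AnimDynamics nodes simulate the flaps.
# A RigidBody node using another cosmetic's physics asset is left unconnected.
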AnimBlueprint /BRCosmetics/Characters/Player/Male/Medium/Heads/M_MED_DairyString_WadePuck_Head/Meshes/M_MED_DairyString_WadePuck_Head_PostProcess_AnimBP
target_skeleton=&/M_MED_DairyString_WadePuck_Head_Skeleton
graphs [
 AnimGraph {nodes [
   RigidBody() {
    bClampLinearTranslationLimitToRefPose=true
    bOverrideWorldGravity=true
    overridePhysicsAsset=/BRCosmetics/Characters/Player/Male/Medium/Bodies/M_MED_BananaAdventure/Meshes/Parts/M_MED_BananaAdventure_FaceAcc_Physics
    simulationSpace=BaseBoneSpace
    alphaBoolBlend {blendOption=Linear}
    baseBoneRef {boneName=spine_05}
    componentAppliedLinearAccClamp=<800 800 800>
    componentLinearAccScale=<0.4 0.4 0.5>
    overrideWorldGravity {z=-980}
   }
   n1 = "Transform (Modify) Bone - Bone: neck_01"(
    ComponentPose=Local To Component(LocalPose=Input Pose())
    Alpha=1
    Translation="0,0,0"
    Rotation=Get neck_01_rot() -> (neck_01_rot:Rotator)
    Scale="1,1,1"
   ) {rotationMode=BMM_Additive, rotationSpace=BCS_BoneSpace, boneToModify {boneName=neck_01}}
   n2 = Get neck_02_rot() -> (neck_02_rot:Rotator)
   n3 = Anim Dynamics - Bone: dyn_flap_01_l(
    ComponentPose="Transform (Modify) Bone - Bone: head"(
     ComponentPose="Transform (Modify) Bone - Bone: neck_02"(ComponentPose=@n1, Alpha=1, Translation="0,0,0", Rotation=@n2, Scale="1,1,1") {rotationMode=BMM_Additive, rotationSpace=BCS_BoneSpace, boneToModify {boneName=neck_02}}
     Alpha=1
     Translation="0,0,0"
     Rotation=Get head rot() -> (head rot:Rotator)
     Scale="1,1,1"
    ) {rotationMode=BMM_Additive, rotationSpace=BCS_BoneSpace, boneToModify {boneName=head}}
    Alpha=1
    ExternalForce="0,0,0"
   ) {
    angularDampingOverride=1
    angularSpringConstant=10000
    bAngularSpring=true
    bOverrideAngularDamping=true
    bUseGravityOverride=true
    gravityScale=0.35
    boundBone {boneName=dyn_flap_01_l}
    componentAppliedLinearAccClamp=<2000 2000 2000>
    componentLinearAccScale=<0.5 0.5 0.5>
    componentLinearVelScale=<0.5 0.5 0.5>
    gravityOverride {z=-980}
    physicsBodyDefinitions [{
      boundBone {boneName=dyn_flap_01_l}
      boxExtents=<20 30 30>
      constraintSetup {angularLimitsMax=<0 15 25>, angularLimitsMin=<0 -15 -80>, angularTarget=<1 0 0>}
      localJointOffset=<4 0 0>
     }]
   }
   n4 = Anim Dynamics - Bone: dyn_flap_02_l(ComponentPose=@n3, Alpha=1, ExternalForce="0,0,0") {
    angularDampingOverride=1
    angularSpringConstant=10000
    bAngularSpring=true
    bOverrideAngularDamping=true
    bUseGravityOverride=true
    gravityScale=0.35
    boundBone {boneName=dyn_flap_02_l}
    componentAppliedLinearAccClamp=<2000 2000 2000>
    componentLinearAccScale=<0.5 0.5 0.5>
    componentLinearVelScale=<0.5 0.5 0.5>
    gravityOverride {z=-980}
    physicsBodyDefinitions [{
      boundBone {boneName=dyn_flap_02_l}
      boxExtents=<20 30 30>
      constraintSetup {angularLimitsMax=<0 15 15>, angularLimitsMin=<0 -15 -80>, angularTarget=<1 0 0>}
      localJointOffset=<4 0 0>
     }]
   }
   n5 = Anim Dynamics - Bone: dyn_flap_r(ComponentPose=@n4, Alpha=1, ExternalForce="0,0,0") {
    angularDampingOverride=1
    angularSpringConstant=9000
    bAngularSpring=true
    bOverrideAngularDamping=true
    bUseGravityOverride=true
    gravityScale=0.35
    boundBone {boneName=dyn_flap_r}
    componentAppliedLinearAccClamp=<2000 2000 2000>
    componentLinearAccScale=<0.2 0.2 0.2>
    componentLinearVelScale=<0.2 0.2 0.2>
    gravityOverride {z=-980}
    physicsBodyDefinitions [{
      boundBone {boneName=dyn_flap_r}
      boxExtents=<20 30 30>
      constraintSetup {angularLimitsMax=<0 15 60>, angularLimitsMin=<0 -15 -1>, angularTarget=<1 0 0>}
      localJointOffset=<4 0 0>
     }]
   }
   Output Pose(Result=Component To Local(ComponentPose=@n5))
  ]}
 EventGraph {nodes [
   Event BlueprintUpdateAnimation()
   > n1 = Sequence!(execute) -> (then_0, then_1, then_2:Exec)
   n2 = GetOwnerSkeletalMeshComponent()
   n3 = BreakRotator(InRot=BreakTransform(InTransform=GetDeltaTransformFromRefPose(self=@n2, BoneName=head, BaseName=neck_02)).Rotation)
   n4 = Get Z percentage() -> (Z percentage:float)
   n5 = Get X percentage() -> (X percentage:float)
   n6 = Get Y percentage() -> (Y percentage:float)
   > Set head rot(head rot:Rotator=MakeRotator(
     Roll=float * float!(A=@n3, B=-1, C:double=@n5) -> (ReturnValue)
     Pitch=float * float(A=@n3.Pitch, B=-1, C=@n6)
     Yaw=float * float(A=@n3.Yaw, B=-1, C=@n4)
    )) -> (Output_Get:Rotator)
   n7 = GetOwnerSkeletalMeshComponent()
   n8 = BreakRotator(InRot=BreakTransform(InTransform=GetDeltaTransformFromRefPose(self=@n7, BoneName=neck_02, BaseName=neck_01)).Rotation)
   n9 = Get X percentage()
   n10 = Get Y percentage()
   n11 = Get Z percentage()
   Set neck_02_rot(
    execute=@n1.then_1
    neck_02_rot:Rotator=MakeRotator(Roll=float * float(A=@n8, B=-1, C=@n9), Pitch=float * float(A=@n8.Pitch, B=-1, C=@n10), Yaw=float * float(A=@n8.Yaw, B=-1, C=@n11))
   ) -> (Output_Get:Rotator)
   n12 = GetOwnerSkeletalMeshComponent()
   n13 = BreakRotator(InRot=BreakTransform(InTransform=GetDeltaTransformFromRefPose(self=@n12, BoneName=neck_01, BaseName=spine_05)).Rotation)
   n14 = Get Z percentage()
   n15 = Get X percentage()
   n16 = Get Y percentage()
   Set neck_01_rot(
    execute=@n1.then_2
    neck_01_rot:Rotator=MakeRotator(Roll=float * float(A=@n13, B=-1, C=@n15), Pitch=float * float(A=@n13.Pitch, B=-1, C=@n16), Yaw=float * float(A=@n13.Yaw, B=-1, C=@n14))
   ) -> (Output_Get:Rotator)
   TryGetPawnOwner()
  ]}
]
variables [
 head rot {type=Rotator}
 neck_01_rot {type=Rotator}
 neck_02_rot {type=Rotator}
 X percentage {type=number, default=0.9}
 Y percentage {type=number, default=0.9}
 Z percentage {type=number, default=0.9}
]
